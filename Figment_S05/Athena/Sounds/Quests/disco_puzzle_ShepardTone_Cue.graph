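# Sound cue for the disco puzzle quest's Shepard tone in Figment S05 Athena content.
# A single looping wave player plays the disco_sheperd_A wave directly as the cue root.
# The cue overrides attenuation for distance-based falloff and routes through the Fort_DanceMusic_Quest_PSM sound class at slightly reduced volume.
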SoundCue /Figment_S05/Athena/Sounds/Quests/disco_puzzle_ShepardTone_Cue
root=WavePlayer_1
nodes [{id=WavePlayer_1, properties {bLooping=true, soundWaveAssetPtr=&/disco_sheperd_A}}]
properties {
 bOverrideAttenuation=true
 maxDistance=3000
 soundClassObject=/Game/Sounds/FortSoundClassesAndModes/Fort_DanceMusic_Quest_PSM
 volumeMultiplier=0.8
 attenuationOverrides {falloffDistance=2000, attenuationShapeExtents {x=1000}}
}
