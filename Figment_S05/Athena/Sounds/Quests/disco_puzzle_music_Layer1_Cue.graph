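# Sound cue for the first layer of looping disco puzzle music, routed to a dance music quest sound class.
# A mixer plays two looping waves together, the main track and a reverb version, each through its own attenuation.
# The main track uses inside dance floor attenuation overridden with a box shape and stereo spread, while the reverb version uses distant attenuation.
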SoundCue /Figment_S05/Athena/Sounds/Quests/disco_puzzle_music_Layer1_Cue
root=Mixer_0
nodes [
 {
  id=Attenuation_0
  children [WavePlayer_0]
  properties {
   attenuationSettings=/Figment_S05/Sounds/Attenuation/Fort_DanceFloorMusic_3x3_Inside_Athena_Att1
   attenuationOverrides {attenuationShape=Box, stereoSpread=10000, attenuationShapeExtents=<1536 1280 748>}
  }
 }
 {
  id=Attenuation_1
  children [WavePlayer_1]
  properties {attenuationSettings=/Figment_S05/Sounds/Attenuation/Fort_DanceFloorMusic_3x3_Distant_Athena_Att}
 }
 {id=Mixer_0, children [Attenuation_0, Attenuation_1], properties {inputVolume [1,1]}}
 {id=WavePlayer_0, properties {bLooping=true, soundWaveAssetPtr=&/disco_puzzle_music_Layer1}}
 {id=WavePlayer_1, properties {bLooping=true, soundWaveAssetPtr=&/disco_puzzle_music_Layer1_reverb}}
]
properties {maxDistance=3036, soundClassObject=/Game/Sounds/FortSoundClassesAndModes/Fort_DanceMusic_Quest_PSM}
