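# Sound cue that loops the second layer of the disco puzzle music for a Figment Season 5 Athena quest.
# A single looping wave player feeds the output directly, with no mixing or modulation nodes.
# Playback uses an indoor dance-floor music attenuation asset with a short max distance and routes through the quest dance music sound class.
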SoundCue /Figment_S05/Athena/Sounds/Quests/disco_puzzle_music_Layer2_Cue
root=WavePlayer_0
nodes [{id=WavePlayer_0, properties {bLooping=true, soundWaveAssetPtr=&/disco_puzzle_music_Layer2}}]
properties {
 attenuationSettings=/Figment_S05/Sounds/Attenuation/Fort_DanceFloorMusic_3x3_Inside_Athena_Att1
 maxDistance=1500
 soundClassObject=/Game/Sounds/FortSoundClassesAndModes/Fort_DanceMusic_Quest_PSM
}
